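# Niagara system for the Zest back bling cosmetic: an audio-reactive, rotating mesh core with a colored light, wrapped in glows, lines and sparks.
# System.Debug_AudioReact Spectrum and Amp values drive core scale and material parameters, light color, color-band tints, noise strength, quad spawning and Flash bursts.
# Colorful quads blend back toward stored spawn positions by spectrum level; GPU child sprites spawn from ParentEmitter particles via attribute reads.
# User.GlobalScale sizes most emitters, and front-end modifiers retune core material parameters.
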NiagaraSystem /BRCosmetics/Accessories/FORT_Backpacks/Backpack_F_MED_Zest/FX/FX/NS_BP_F_Commando_Zest
dependencies {
 usedDataInterfaces [
  &NiagaraDataInterfaceCurve
  &NiagaraDataInterfaceCamera
  &NiagaraDataInterfaceColorCurve
  &NiagaraDataInterfaceParticleRead
  &NiagaraDataInterfaceVector2DCurve
 ]
 usedDynamicInputs [
  &MakeQuatFromAxisAngle
  &Lerp_Float
  &Waveform
  /Niagara/DynamicInputs/LinearColor/MakeLinearColorFromFloat4
  &MakeVector4
 ]
 usedModules [
  &self:SystemSpawnScript.SystemScriptSource.SystemScriptGraph.NiagaraNodeAssignment_0.SetVariables_#1
  &self:SystemSpawnScript.SystemScriptSource.SystemScriptGraph.NiagaraNodeAssignment_2.SetVariables_#2
  &SystemState
  &self:SystemSpawnScript.SystemScriptSource.SystemScriptGraph.NiagaraNodeAssignment_1.SetVariables_#3
  &self:SystemSpawnScript.SystemScriptSource.SystemScriptGraph.NiagaraNodeAssignment_4.SetVariables_#4
  &self:SystemSpawnScript.SystemScriptSource.SystemScriptGraph.NiagaraNodeAssignment_5.SetVariables_#5
 ]
}
emitters {
 ChildEmitter {
  simTarget=GPUComputeSim
  emitterUpdateScript [
   &EmitterState(Life Cycle Mode=::NewEnumerator1, Loop Duration=2, UseLoopDelay=false)
   &SpawnRate(SpawnRate=20, Use Spawn Probability=false) {enabled=false}
   &SpawnParticlesFromOtherEmitter(
    Attribute Reader=DataInterface() {propertyValues {EmitterBinding {bindingMode=Other, emitterName=ParentEmitter}}}
    Spawn Rate=6
    Use Maximum Age=false
    Use Minimum Age=false
   )
  ]
  particleSpawnScript [
   &InitializeParticle(
    Color Channel Mode:ENiagaraChannelCorrelation=::NewEnumerator2
    Color Maximum=<rgba 1 1 1 0.6>
    Color Minimum=<rgba 0 1.9437 2 0.75>
    Color Mode=::NewEnumerator2
    Lifetime Max=0.7
    Lifetime Min=0.6
    Lifetime Mode=::NewEnumerator1
    Sprite Size Mode=::NewEnumerator3
    Uniform Sprite Size=1.6
    UsePositionOffset=false
    Write Material Random=true
   )
   &SampleParticlesFromOtherEmitter(Sprite Size Sampling:ENiagara_DirectReadSamplingMode=::NewEnumerator0)
   /SetVariables(Particles.Color=HlslExpression() {hlslExpression="Particles.Color * float4(1.0, 1.0, 1.0, rand(0.5)+0.5) /* Custom HLSL! */"}) {isSetParameters=true, module=SetVariables_#6}
   &SphereLocation(
    MaskSpawnGroup=false
    Non Uniform Scale=<1 0 0>
    Override Seed=false
    Sphere Radius=0.5
    Surface Only=false
    UseNonUniformScale=true
    UseOffset=false
    UseSphereOrientation=false
   )
   &AddVelocity(Velocity=&Multiply_Vector(
     A=<0 0 9>
     B=&VectorFromFloat(Value=&MakeFloatFromVector2d(Vector 2D=$Particles.SpriteSize))
    ))
  ]
  particleUpdateScript [
   &ParticleState()
   &Drag(Drag=0.25, Use Linear Drag=true, Use Rotational Drag=false)
   &CurlNoiseForce(
    Noise Strength=&Multiply_Float(
     A=&MakeFloatFromLinearColor(LinearColor=$System.Debug_AudioReact_Spectrum, Channel:ENiagaraLinearColor_Channels=::NewEnumerator3)
     B=500
    )
    Pan Noise=false
   )
   &ScaleColor(
    Scale Alpha=&FloatFromCurve(FloatCurve=DataInterface() {propertyValues {Curve {keys [<at -0.0046 1.0113 RCIM_Cubic>,<at 1.0011 0.002 RCIM_Cubic>]}}})
    Scale RGB=&Multiply_Vector(
     A=<1 1 1>
     B=&MakeVectorFromLinearColorRGB(Color=&ColorFromCurve(
       ColorCurve=DataInterface() {propertyValues {
         AlphaCurve {keys [<at 0 1 RCIM_Constant>]}
         BlueCurve {keys [<at 0 1 RCIM_Constant>,<at 0.1198 1 RCIM_Constant>,<at 0.2388 0.0651 RCIM_Constant>,<at 0.3516 0.2 RCIM_Constant>,<at 0.4972 0 RCIM_Constant>,<at 0.6241 1 RCIM_Constant>,<at 0.7666 0 RCIM_Constant>,<at 0.8872 0.847 RCIM_Constant>]}
         GreenCurve {keys [<at 0 1 RCIM_Constant>,<at 0.1198 0.0289 RCIM_Constant>,<at 0.2388 0 RCIM_Constant>,<at 0.3516 1 RCIM_Constant>,<at 0.4972 0.9308 RCIM_Constant>,<at 0.6241 0 RCIM_Constant>,<at 0.7666 1 RCIM_Constant>,<at 0.8872 0 RCIM_Constant>]}
         RedCurve {keys [<at 0 1 RCIM_Constant>,<at 0.1198 0 RCIM_Constant>,<at 0.2388 1 RCIM_Constant>,<at 0.3516 0 RCIM_Constant>,<at 0.4972 1 RCIM_Constant>,<at 0.6241 0.2068 RCIM_Constant>,<at 0.7666 0.0984 RCIM_Constant>,<at 0.8872 1 RCIM_Constant>]}
        }}
       CurveIndex=&MakeFloatFromLinearColor(LinearColor=$System.Debug_AudioReact_Spectrum, Channel:ENiagaraLinearColor_Channels=::NewEnumerator1)
      ))
    )
    ScaleA=false
    ScaleRGB=true
   )
   &ScaleSpriteSize(
    Uniform Curve Index=$Emitter.LoopedAge
    Uniform Curve Sprite Scale=DataInterface() {propertyValues {Curve {keys [
        {arriveTangent=7.4781, interpMode=RCIM_Cubic, leaveTangent=7.4781, tangentMode=RCTM_User, time=0, value=0.7}
        {interpMode=RCIM_Cubic, tangentMode=RCTM_User, time=0.3, value=1}
        {tangentMode=RCTM_User, time=1, value=0.7}
       ]}}}
   )
   &SolveForcesAndVelocity(Clamp Velocity=false, Limit Acceleration=false)
   &CameraOffset(Camera Offset Amount=7)
  ]
  rendererClasses [&NiagaraSpriteRendererProperties]
  renderers [{rendererClass=&NiagaraSpriteRendererProperties}]
 }
 ChildEmitter001 {
  emitterUpdateScript [
   &EmitterState(Life Cycle Mode=::NewEnumerator1, Loop Duration=2, UseLoopDelay=false)
   &SpawnRate(SpawnRate=10, Use Spawn Probability=false)
  ]
  particleSpawnScript [
   &InitializeParticle(
    Color Channel Mode:ENiagaraChannelCorrelation=::NewEnumerator2
    Color Maximum=<rgba 1 1 1 0.6>
    Color Minimum=<rgba 0 1.9437 2 0.75>
    Color Mode=::NewEnumerator2
    Lifetime Max=0.7
    Lifetime Min=0.6
    Lifetime Mode=::NewEnumerator1
    Sprite Size Mode=::NewEnumerator3
    Uniform Sprite Size=1.6
    UsePositionOffset=false
    Write Material Random=true
   )
   /SetVariables(
    Particles.Color=HlslExpression() {hlslExpression="Particles.Color * float4(1.0, 1.0, 1.0, rand(0.5)+0.5) /* Custom HLSL! */"}
    Particles.initRand=HlslExpression() {hlslExpression="sign(rand(1.0)-0.5)"}
   ) {isSetParameters=true, module=SetVariables_#6}
   &SphereLocation(
    MaskSpawnGroup=false
    Non Uniform Scale=<50 0 18>
    Override Seed=false
    Sphere Radius=0.5
    Surface Only=false
    UseNonUniformScale=true
    UseOffset=false
    UseSphereOrientation=false
   )
   &AddVelocity(Velocity=&Multiply_Vector(
     A=&Multiply_VectorByFloat(Vector=<0 0 9>, Float=$Particles.initRand)
     B=&VectorFromFloat(Value=&MakeFloatFromVector2d(Vector 2D=$Particles.SpriteSize))
    ))
  ]
  particleUpdateScript [
   &ParticleState()
   &Drag(Drag=0.25, Use Linear Drag=true, Use Rotational Drag=false)
   &CurlNoiseForce(
    Noise Strength=&Multiply_Float(
     A=&MakeFloatFromLinearColor(LinearColor=$System.Debug_AudioReact_Spectrum, Channel:ENiagaraLinearColor_Channels=::NewEnumerator3)
     B=500
    )
    Pan Noise=false
   )
   &ScaleColor(
    Scale Alpha=&FloatFromCurve(FloatCurve=DataInterface() {propertyValues {Curve {keys [<at -0.0046 1.0113 RCIM_Cubic>,<at 1.0011 0.002 RCIM_Cubic>]}}})
    Scale RGB=&Multiply_Vector(
     A=<1 1 1>
     B=&MakeVectorFromLinearColorRGB(Color=&ColorFromCurve(
       ColorCurve=DataInterface() {propertyValues {
         AlphaCurve {keys [<at 0 1 RCIM_Constant>]}
         BlueCurve {keys [<at 0 1 RCIM_Constant>,<at 0.1198 1 RCIM_Constant>,<at 0.2388 0.0651 RCIM_Constant>,<at 0.3516 0.2 RCIM_Constant>,<at 0.4972 0 RCIM_Constant>,<at 0.6241 1 RCIM_Constant>,<at 0.7666 0 RCIM_Constant>,<at 0.8872 0.847 RCIM_Constant>]}
         GreenCurve {keys [<at 0 1 RCIM_Constant>,<at 0.1198 0.0289 RCIM_Constant>,<at 0.2388 0 RCIM_Constant>,<at 0.3516 1 RCIM_Constant>,<at 0.4972 0.9308 RCIM_Constant>,<at 0.6241 0 RCIM_Constant>,<at 0.7666 1 RCIM_Constant>,<at 0.8872 0 RCIM_Constant>]}
         RedCurve {keys [<at 0 1 RCIM_Constant>,<at 0.1198 0 RCIM_Constant>,<at 0.2388 1 RCIM_Constant>,<at 0.3516 0 RCIM_Constant>,<at 0.4972 1 RCIM_Constant>,<at 0.6241 0.2068 RCIM_Constant>,<at 0.7666 0.0984 RCIM_Constant>,<at 0.8872 1 RCIM_Constant>]}
        }}
       CurveIndex=&MakeFloatFromLinearColor(LinearColor=$System.Debug_AudioReact_Spectrum, Channel:ENiagaraLinearColor_Channels=::NewEnumerator1)
      ))
    )
    ScaleA=false
    ScaleRGB=true
   )
   &ScaleSpriteSize(
    Uniform Curve Index=$Emitter.LoopedAge
    Uniform Curve Sprite Scale=DataInterface() {propertyValues {Curve {keys [
        {arriveTangent=7.4781, interpMode=RCIM_Cubic, leaveTangent=7.4781, tangentMode=RCTM_User, time=0, value=0.7}
        {interpMode=RCIM_Cubic, tangentMode=RCTM_User, time=0.3, value=1}
        {tangentMode=RCTM_User, time=1, value=0.7}
       ]}}}
   )
   &SolveForcesAndVelocity(Clamp Velocity=false, Limit Acceleration=false)
   &CameraOffset(Camera Offset Amount=7)
  ]
  rendererClasses [&NiagaraSpriteRendererProperties]
  renderers [{rendererClass=&NiagaraSpriteRendererProperties}]
 }
 Colorful_Quads {
  emitterUpdateScript [
   &EmitterState(Life Cycle Mode=::NewEnumerator1, Loop Duration=1, UseLoopDelay=false)
   &SpawnBurst_Instantaneous(Spawn Count=100, Use Spawn Probability=false, UseLoopCountLimit=false) {enabled=false}
   &SpawnRate(
    SpawnRate=&Multiply_Float(
     A=95
     B=&MakeFloatFromLinearColor(LinearColor=$System.Debug_AudioReact_Spectrum, Channel:ENiagaraLinearColor_Channels=::NewEnumerator3)
    )
    Use Spawn Probability=false
   )
   &SpawnRate(SpawnRate=100, Use Spawn Probability=false) {enabled=false, module=SpawnRate001}
  ]
  particleSpawnScript [
   &InitializeParticle(
    Lifetime Max=1
    Lifetime Min=0.5
    Lifetime Mode=::NewEnumerator1
    Sprite Size Mode=::NewEnumerator1
    Uniform Sprite Size Max=2.6
    Uniform Sprite Size Min=1.7
    UsePositionOffset=false
    Write Material Random=true
   )
   /Niagara/Modules/Update/Mass/CalculateSizeAndRotationalInertiaByMass(Module.Manually Enter Density=false)
   &SphereLocation(
    MaskSpawnGroup=false
    Non Uniform Scale=<1 0.4 0.8>
    Override Seed=false
    Sphere Radius=&Multiply_Float(A=40, B=$User.GlobalScale)
    Surface Only=false
    UseNonUniformScale=true
    UseOffset=false
    UseSphereOrientation=false
   )
   &AddVelocityFromPoint(
    Module.Use Falloff Distance=false
    Origin Offset=<0 -1 -7>
    Velocity Origin:Vector3f=/Niagara/DynamicInputs/Helpers/SimulationPosition()
    Velocity Strength=&RandomRangeFloat(Minimum=300, Maximum=400)
   )
   /SetVariables(Particles.StoredPos:NiagaraPosition=$Particles.Position) {isSetParameters=true}
  ]
  particleUpdateScript [
   &ParticleState()
   &Drag(Drag=8, Use Linear Drag=true, Use Rotational Drag=false)
   &ScaleColor(
    Scale Alpha=&FloatFromCurve(FloatCurve=DataInterface() {propertyValues {Curve {keys [<at -0.0046 1.0113 RCIM_Cubic>,<at 1.0011 0.002 RCIM_Cubic>]}}}, Scale Curve=0.5)
    Scale RGB=&Multiply_Vector(
     A=<1 1 1>
     B=&MakeVectorFromLinearColorRGB(Color=&ColorFromCurve(
       ColorCurve=DataInterface() {propertyValues {
         AlphaCurve {keys [<at 0 1 RCIM_Constant>]}
         BlueCurve {keys [<at 0 0.8 RCIM_Constant>,<at 0.0579 1 RCIM_Constant>,<at 0.1738 0.2744 RCIM_Constant>,<at 0.3171 0.3667 RCIM_Constant>,<at 0.4592 0.2396 RCIM_Constant>,<at 0.6241 1 RCIM_Constant>,<at 0.7666 0.1842 RCIM_Constant>,<at 0.8872 0.8948 RCIM_Constant>]}
         GreenCurve {keys [<at 0 0.8 RCIM_Constant>,<at 0.0579 0.1654 RCIM_Constant>,<at 0.1738 0.224 RCIM_Constant>,<at 0.3171 1 RCIM_Constant>,<at 0.4592 0.9474 RCIM_Constant>,<at 0.6241 0.1302 RCIM_Constant>,<at 0.7666 1 RCIM_Constant>,<at 0.8872 0.3125 RCIM_Constant>]}
         RedCurve {keys [<at 0 0.8 RCIM_Constant>,<at 0.0579 0.1406 RCIM_Constant>,<at 0.1738 1 RCIM_Constant>,<at 0.3171 0.2083 RCIM_Constant>,<at 0.4592 1 RCIM_Constant>,<at 0.6241 0.3101 RCIM_Constant>,<at 0.7666 0.5224 RCIM_Constant>,<at 0.8872 1 RCIM_Constant>]}
        }}
       CurveIndex=&MakeFloatFromLinearColor(LinearColor=$System.Debug_AudioReact_Spectrum, Channel:ENiagaraLinearColor_Channels=::NewEnumerator1)
      ))
    )
    ScaleA=true
    ScaleRGB=true
   )
   &ScaleSpriteSize(Scale Factor=&Vector2DFromFloat(Value=&FloatFromCurve(
      FloatCurve=DataInterface() {propertyValues {Curve {keys [
          <at -0.0046 0 RCIM_Cubic>
          <at 0.1 1>
          {arriveTangent=-1.163, interpMode=RCIM_Cubic, leaveTangent=-1.163, time=0.7001, value=0.872}
          <at 1.0011 0.002>
         ]}}}
      Scale Curve=0.4
     )))
   &SolveForcesAndVelocity(Clamp Velocity=false, Limit Acceleration=false)
   /SetVariables(Particles.Position=&Lerp_Vector(
     A=Linked() {linkedVariable {name=Particles.Position, type=NiagaraPosition}}
     B=$Particles.StoredPos
     Alpha=&MakeFloatFromLinearColor(LinearColor=$System.Debug_AudioReact_Spectrum, Channel:ENiagaraLinearColor_Channels=::NewEnumerator3)
    )) {isSetParameters=true}
  ]
  rendererClasses [&NiagaraSpriteRendererProperties]
  renderers [{rendererClass=&NiagaraSpriteRendererProperties}]
 }
 Core {
  emitterUpdateScript [
   &EmitterState(
    Inactive Response=::NewEnumerator1
    Life Cycle Mode=::NewEnumerator1
    Loop Behavior=::NewEnumerator1
    Loop Duration Mode=::NewEnumerator1
    UseLoopDelay=false
   )
   &SpawnBurst_Instantaneous(Use Spawn Probability=false, UseLoopCountLimit=false)
  ]
  particleSpawnScript [
   &InitializeParticle(
    Color=<rgba 0.4056 0.2913 0.4323 0.8>
    Lifetime=5
    Mesh Scale=&Multiply_VectorByFloat(Vector=<0.627 1 0.32>, Float=$User.GlobalScale)
    Mesh Scale Mode=::NewEnumerator4
    Position Offset=<0 -7 0>
    UsePositionOffset=true
    Write Material Random=true
   )
   &DynamicMaterialParameters(
    Index 0 Param 1=$System.Random
    Index 0 Param 2=&MakeFloatFromVector(Vector=$Particles.Scale, Channel=::NewEnumerator0)
    Index 0 Param 3=&MakeFloatFromLinearColor(LinearColor=$System.Debug_AudioReact_Spectrum, Channel:ENiagaraLinearColor_Channels=::NewEnumerator1)
    Module.Param0WriteEnabled=true
    Module.Param1WriteEnabled=true
    Module.Param2WriteEnabled=true
    Module.Param3WriteEnabled=false
   )
   /SetVariables(Particles.Position=&Add_Vector(A=Linked() {linkedVariable {name=Particles.Position, type=NiagaraPosition}}, B=<0 5 0>)) {isSetParameters=true, module=SetVariables_#7}
  ]
  particleUpdateScript [
   &ParticleState(Kill Particles When Lifetime Has Elapsed=false, Let Infinitely Lived Particles Die When Emitter Deactivates=true)
   /SetVariables(Particles.MeshOrientation=$System.AutoRotate) {isSetParameters=true, module=SetVariables_#8}
   &DynamicMaterialParameters(
    Index 0 Param 1=$System.Random
    Index 0 Param 2=&MakeFloatFromVector(Vector=$Particles.Scale, Channel=::NewEnumerator0)
    Index 0 Param 3=&MakeFloatFromLinearColor(LinearColor=$System.Debug_AudioReact_Spectrum, Channel:ENiagaraLinearColor_Channels=::NewEnumerator3)
    Index 0 Param 4=&FloatFromCurve(
     FloatCurve=DataInterface() {propertyValues {Curve {keys [<at 0 0>,<at 0.1 0>,<at 0.1615 1>,<at 0.1932 -0.375>,<at 0.279 0>,<at 0.4 0.125>,<at 0.5 0.625>,<at 0.5773 -0.875>,<at 0.6232 0.125>,<at 0.693 -0.5>,<at 0.8 0.475>,<at 0.9 -0.2>,<at 1 0>]}}}
     CurveIndex=HlslExpression() {hlslExpression="frac(Emitter.Age * 0.5)"}
     Scale Curve=0.04
    )
    Index 1 Param 1=&FN_FrontEndFloatModifier(Gameplay=0.2, Front End=0.15)
    Index 1 Param 2=&Multiply_Float(
     A=&Add_Float(
      A=2
      B=&Multiply_Float(
       A=&MakeFloatFromLinearColor(LinearColor=$System.Debug_AudioReact_Amp, Channel:ENiagaraLinearColor_Channels=::NewEnumerator0)
       B=5
      )
     )
     B=&MakeFloatFromLinearColor(
      LinearColor=&MultiplyLinearColor(
       Linear Color=/Niagara/DynamicInputs/TypeConversions/MakeLinearColorFromFloat(Value:NiagaraFloat=&MakeFloatFromLinearColor(LinearColor=$System.Debug_AudioReact_Amp, Channel:ENiagaraLinearColor_Channels=::NewEnumerator0))
       Scale Factor=&MakeVector4_FromFloat(Float=2)
      )
      Channel:ENiagaraLinearColor_Channels=::NewEnumerator3
     )
    )
    Index 1 Param 3=&FN_FrontEndFloatModifier(Gameplay=8, Front End=4)
    Index 1 Param 4=1
    Module.Param0WriteEnabled=false
    Module.Param0WriteEnabled1=true
    Module.Param1WriteEnabled=false
    Module.Param1WriteEnabled1=true
    Module.Param2WriteEnabled=true
    Module.Param2WriteEnabled1=true
    Module.Param3WriteEnabled=true
    Module.Param3WriteEnabled1=true
    Write Parameter Index 1=true
   ) {module=DynamicMaterialParameters001}
   &self:LODSetting(LOD1=400, LOD2=1000) {enabled=false, module=LODSetting}
   &ScaleMeshSize(Scale Factor=&Add_Vector(
     A=<1 1 1>
     B=&Multiply_VectorByFloat(Vector=<1.152 1 1.161>, Float=HlslExpression() {hlslExpression="min(0.6, System.Debug_AudioReact_Spectrum.a) * User.GlobalScale"})
    ))
   &FN_Light_Attributes(
    Color=&LerpLinearColors(
     StartColor=&ColorFromCurve(
      ColorCurve=DataInterface() {propertyValues {
        AlphaCurve {keys [<at 0 1 RCIM_Cubic>]}
        BlueCurve {keys [<at 0 1 RCIM_Cubic>,<at 0.2988 1>,<at 0.6244 0.4879>,<at 0.8067 0.9395>,<at 1 1>]}
        GreenCurve {keys [<at 0 0.7391 RCIM_Cubic>,<at 0.2988 0.184>,<at 0.6244 1>,<at 0.8067 0.6452>,<at 1 0.7391>]}
        RedCurve {keys [<at 0 0 RCIM_Cubic>,<at 0.2988 0.5953>,<at 0.6244 0.1225>,<at 0.8067 1>,<at 1 0>]}
       }}
      CurveIndex=HlslExpression() {hlslExpression="frac(Emitter.Age*0.35)"}
     )
     EndColor=&ColorFromCurve(
      ColorCurve=DataInterface() {propertyValues {
        AlphaCurve {keys [<at 0 1 RCIM_Constant>]}
        BlueCurve {keys [<at 0 0.8 RCIM_Constant>,<at 0.0579 1 RCIM_Constant>,<at 0.1738 0.2452 RCIM_Constant>,<at 0.3171 0.3208 RCIM_Constant>,<at 0.4592 0.2396 RCIM_Constant>,<at 0.6241 1 RCIM_Constant>,<at 0.7666 0.1842 RCIM_Constant>,<at 0.8872 0.8948 RCIM_Constant>]}
        GreenCurve {keys [<at 0 0.8 RCIM_Constant>,<at 0.0579 0.1654 RCIM_Constant>,<at 0.1738 0.1927 RCIM_Constant>,<at 0.3171 1 RCIM_Constant>,<at 0.4592 0.9474 RCIM_Constant>,<at 0.6241 0.1302 RCIM_Constant>,<at 0.7666 1 RCIM_Constant>,<at 0.8872 0.3125 RCIM_Constant>]}
        RedCurve {keys [<at 0 0.8 RCIM_Constant>,<at 0.0579 0.1406 RCIM_Constant>,<at 0.1738 1 RCIM_Constant>,<at 0.3171 0.151 RCIM_Constant>,<at 0.4592 1 RCIM_Constant>,<at 0.6241 0.3101 RCIM_Constant>,<at 0.7666 0.5224 RCIM_Constant>,<at 0.8872 1 RCIM_Constant>]}
       }}
      CurveIndex=&MakeFloatFromLinearColor(LinearColor=$System.Debug_AudioReact_Amp, Channel:ENiagaraLinearColor_Channels=::NewEnumerator0)
     )
     LerpFactor=&MakeFloatFromLinearColor(LinearColor=$System.Debug_AudioReact_Spectrum, Channel:ENiagaraLinearColor_Channels=::NewEnumerator3)
    )
    Create Color Attribute=true
    Create Exponent Attribute=true
    Create Radius Attribute=true
    Exponent=3.5
    Radius=25
    Scale Color=12.5
   )
   &CameraOffset(Camera Offset Amount=12)
  ]
  rendererClasses [&NiagaraMeshRendererProperties, &NiagaraLightRendererProperties]
  renderers [
   {rendererClass=&NiagaraMeshRendererProperties, rendererIndex=0}
   {rendererClass=&NiagaraLightRendererProperties, rendererIndex=1}
  ]
 }
 Core002 {
  emitterUpdateScript [
   &EmitterState(
    Inactive Response=::NewEnumerator1
    Life Cycle Mode=::NewEnumerator1
    Loop Behavior=::NewEnumerator1
    Loop Duration Mode=::NewEnumerator1
    UseLoopDelay=false
   )
   &SpawnBurst_Instantaneous(Use Spawn Probability=false, UseLoopCountLimit=false)
   &SpawnBurst_Instantaneous(Spawn Time=0.9664, Use Spawn Probability=false, UseLoopCountLimit=false) {module=SpawnBurst_Instantaneous001}
  ]
  particleSpawnScript [
   &InitializeParticle(
    Color=<rgba 5 1.1979 2.6516 0.15>
    Lifetime=5
    Mesh Scale=&Multiply_VectorByFloat(Vector=<0.627 2 0.58>, Float=$User.GlobalScale)
    Mesh Scale Mode=::NewEnumerator4
    Position Offset=<0 -7 0>
    UsePositionOffset=true
    Write Material Random=true
   )
   &DynamicMaterialParameters(
    Index 0 Param 1=$System.Random
    Index 0 Param 2=&MakeFloatFromVector(Vector=$Particles.Scale, Channel=::NewEnumerator0)
    Index 0 Param 3=&MakeFloatFromLinearColor(LinearColor=$System.Debug_AudioReact_Spectrum, Channel:ENiagaraLinearColor_Channels=::NewEnumerator1)
    Module.Param0WriteEnabled=true
    Module.Param1WriteEnabled=true
    Module.Param2WriteEnabled=true
    Module.Param3WriteEnabled=false
   )
   /SetVariables(Particles.Position=&Add_Vector(A=Linked() {linkedVariable {name=Particles.Position, type=NiagaraPosition}}, B=<0 5 0>)) {isSetParameters=true, module=SetVariables_#7}
  ]
  particleUpdateScript [
   &ParticleState(Kill Particles When Lifetime Has Elapsed=false, Let Infinitely Lived Particles Die When Emitter Deactivates=true)
   /SetVariables(Particles.Position=&Add_Vector(
     A=Linked() {linkedVariable {name=Particles.Position, type=NiagaraPosition}}
     B=&Lerp_Vector(A=<0 0 0>, B=<0 0 -10>, Alpha=0)
    )) {isSetParameters=true}
   /SetVariables(Particles.MeshOrientation=$System.AutoRotate) {isSetParameters=true, module=SetVariables_#8}
   &DynamicMaterialParameters(
    Index 0 Param 1=$System.Random
    Index 0 Param 2=&MakeFloatFromVector(Vector=$Particles.Scale, Channel=::NewEnumerator0)
    Index 0 Param 3=&MakeFloatFromLinearColor(LinearColor=$System.Debug_AudioReact_Spectrum, Channel:ENiagaraLinearColor_Channels=::NewEnumerator3)
    Index 1 Param 1=0.1
    Index 1 Param 2=&Multiply_Float(
     A=&Add_Float(
      A=10
      B=&Multiply_Float(
       A=&MakeFloatFromLinearColor(LinearColor=$System.Debug_AudioReact_Amp, Channel:ENiagaraLinearColor_Channels=::NewEnumerator0)
       B=5
      )
     )
     B=&MakeFloatFromLinearColor(
      LinearColor=&MultiplyLinearColor(
       Linear Color=&MakeLinearColorFromFloat(Value:NiagaraFloat=&MakeFloatFromLinearColor(LinearColor=$System.Debug_AudioReact_Amp, Channel:ENiagaraLinearColor_Channels=::NewEnumerator0))
       Scale Factor=&MakeVector4_FromFloat(Float=4)
      )
      Channel:ENiagaraLinearColor_Channels=::NewEnumerator3
     )
    )
    Index 1 Param 3=&FN_FrontEndFloatModifier(Gameplay=8, Front End=4)
    Index 1 Param 4=1
    Module.Param0WriteEnabled=false
    Module.Param0WriteEnabled1=true
    Module.Param1WriteEnabled=false
    Module.Param1WriteEnabled1=true
    Module.Param2WriteEnabled=true
    Module.Param2WriteEnabled1=true
    Module.Param3WriteEnabled=true
    Module.Param3WriteEnabled1=true
    Write Parameter Index 1=true
   ) {module=DynamicMaterialParameters001}
   &self:LODSetting(LOD1=400, LOD2=1000) {enabled=false, module=LODSetting}
   &ScaleMeshSize(Scale Factor=&Add_Vector(
     A=<1 1 1>
     B=&Multiply_VectorByFloat(Vector=<1.152 1 1.161>, Float=HlslExpression() {hlslExpression="min(0.6, System.Debug_AudioReact_Spectrum.a) * User.GlobalScale"})
    ))
   &CameraOffset(Camera Offset Amount=12)
  ]
  rendererClasses [&NiagaraMeshRendererProperties]
  renderers [{rendererClass=&NiagaraMeshRendererProperties, rendererIndex=0}]
 }
 Flash {
  emitterUpdateScript [
   &EmitterState(Life Cycle Mode=::NewEnumerator1, Loop Duration=1, UseLoopDelay=false)
   &SpawnRate(
    SpawnRate=&Multiply_Float(A=24, B=HlslExpression() {hlslExpression="sign(System.Debug_AudioReact_Spectrum.x -0.28)"})
    Use Spawn Probability=false
   )
  ]
  particleSpawnScript [
   &InitializeParticle(
    Color=&ColorFromCurve(
     ColorCurve=DataInterface() {propertyValues {
       AlphaCurve {keys [<at 0 1 RCIM_Constant>]}
       BlueCurve {keys [<at 0 0.8 RCIM_Constant>,<at 0.0579 1 RCIM_Constant>,<at 0.1738 0.3134 RCIM_Constant>,<at 0.3171 0.3292 RCIM_Constant>,<at 0.4592 0.2396 RCIM_Constant>,<at 0.6241 1 RCIM_Constant>,<at 0.7666 0.1842 RCIM_Constant>,<at 0.8872 0.8948 RCIM_Constant>]}
       GreenCurve {keys [<at 0 0.8 RCIM_Constant>,<at 0.0579 0.1654 RCIM_Constant>,<at 0.1738 0.2656 RCIM_Constant>,<at 0.3171 1 RCIM_Constant>,<at 0.4592 0.9474 RCIM_Constant>,<at 0.6241 0.1979 RCIM_Constant>,<at 0.7666 1 RCIM_Constant>,<at 0.8872 0.3125 RCIM_Constant>]}
       RedCurve {keys [<at 0 0.8 RCIM_Constant>,<at 0.0579 0.1406 RCIM_Constant>,<at 0.1738 1 RCIM_Constant>,<at 0.3171 0.1615 RCIM_Constant>,<at 0.4592 1 RCIM_Constant>,<at 0.6241 0.3638 RCIM_Constant>,<at 0.7666 0.5224 RCIM_Constant>,<at 0.8872 1 RCIM_Constant>]}
      }}
     CurveIndex=&RandomRangeFloat(Minimum=0, Maximum=1)
    )
    Lifetime Max=0.4
    Lifetime Min=0.35
    Lifetime Mode=::NewEnumerator1
    Sprite Rotation Mode=::NewEnumerator2
    Sprite Size Mode=::NewEnumerator3
    Uniform Sprite Size=&Multiply_Float(A=100, B=$User.GlobalScale)
    UsePositionOffset=false
    Write Material Random=true
   )
   &CalculateSizeAndRotationalInertiaByMass(Module.Manually Enter Density=false)
   &TorusLocation(
    Align Torus With Axis=true
    Handle Radius=10
    Large Radius:NiagaraFloat=&Multiply_Float(A=30, B=$User.GlobalScale)
    MaskSpawnGroup=false
    Module.Override Randomness=false
    Non Uniform Scale:Vector3f=<1 1 0.2>
    Torus Axis=<0 1 0>
   )
  ]
  particleUpdateScript [
   &ParticleState()
   &Drag(Drag=4, Use Linear Drag=true, Use Rotational Drag=false)
   &ScaleSpriteSize(Scale Factor=&Vector2DFromFloat(Value=&FloatFromCurve(
      FloatCurve=DataInterface() {propertyValues {Curve {keys [
          <at -0.0046 0.12 RCIM_Cubic>
          {arriveTangent=-0.1174, interpMode=RCIM_Cubic, leaveTangent=-0.1174, time=0.3105, value=1}
          <at 1.0011 0.002 RCIM_Cubic>
         ]}}}
      Scale Curve=12.8
     )))
   &ScaleColor(
    Scale Alpha=&FloatFromCurve(FloatCurve=DataInterface() {propertyValues {Curve {keys [<at -0.0046 1.0113 RCIM_Cubic>,<at 1.0011 0.002 RCIM_Cubic>]}}}, Scale Curve=0.3)
    ScaleA=true
    ScaleRGB=true
   )
   &SolveForcesAndVelocity(Clamp Velocity=false, Limit Acceleration=false)
   &CameraOffset()
  ]
  rendererClasses [&NiagaraSpriteRendererProperties]
  renderers [{rendererClass=&NiagaraSpriteRendererProperties}]
 }
 LineSprite {
  emitterUpdateScript [
   &EmitterState(Life Cycle Mode=::NewEnumerator1, Loop Duration=2, UseLoopDelay=false)
   &SpawnRate(SpawnRate=8, Use Spawn Probability=false)
  ]
  particleSpawnScript [
   &InitializeParticle(
    Color Channel Mode:ENiagaraChannelCorrelation=::NewEnumerator2
    Color Maximum=<rgba 0.4 0.2182 0.3593 0.5>
    Color Minimum=<rgba 0 1.9437 2 0.5>
    Color Mode=::NewEnumerator2
    Lifetime Max=1.1
    Lifetime Min=0.9
    Lifetime Mode=::NewEnumerator1
    Sprite Size Max=&Multiply_Vector2D_ByFloat(Float=$User.GlobalScale, Vector2D=<0.5 30>)
    Sprite Size Min=&Multiply_Vector2D_ByFloat(Float=$User.GlobalScale, Vector2D=<0.35 15>)
    Sprite Size Mode=::NewEnumerator2
    UsePositionOffset=false
    Write Material Random=true
   )
   &SphereLocation(
    MaskSpawnGroup=false
    Non Uniform Scale=<4 0 2.2>
    Override Seed=false
    Sphere Radius=&Multiply_Float(A=&Multiply_Float(A=8, B=$User.GlobalScale), B=$User.GlobalScale)
    Surface Only=false
    UseNonUniformScale=true
    UseOffset=false
    UseSphereOrientation=false
   )
  ]
  particleUpdateScript [
   &ParticleState()
   &SolveForcesAndVelocity(Clamp Velocity=false, Limit Acceleration=false)
   &ScaleSpriteSize(
    Non-Uniform Curve Sprite Scale=DataInterface() {propertyValues {
      XCurve {keys [{tangentMode=RCTM_User, time=0, value=1}, {tangentMode=RCTM_User, time=1, value=0.5}]}
      YCurve {keys [{tangentMode=RCTM_User, time=0, value=0.3}, {tangentMode=RCTM_User, time=1, value=1}]}
     }}
    Scale Sprite Size Mode=::NewEnumerator3
   )
   &ScaleColor(
    Scale Alpha=&FloatFromCurve(FloatCurve=DataInterface() {propertyValues {Curve {keys [
         {arriveTangent=5.3877, interpMode=RCIM_Cubic, leaveTangent=5.3877, tangentMode=RCTM_User, time=0, value=0}
         {interpMode=RCIM_Cubic, tangentMode=RCTM_User, time=0.5, value=1}
         {tangentMode=RCTM_User, time=1, value=0}
        ]}}})
    ScaleA=true
    ScaleRGB=false
   )
   &CameraOffset(Camera Offset Amount=-10)
  ]
  rendererClasses [&NiagaraSpriteRendererProperties]
  renderers [{rendererClass=&NiagaraSpriteRendererProperties}]
 }
 ParentEmitter {
  simTarget=GPUComputeSim
  emitterUpdateScript [
   &EmitterState(Life Cycle Mode=::NewEnumerator1, Loop Duration=1, UseLoopDelay=false)
   &SpawnRate(SpawnRate=5, Use Spawn Probability=false)
  ]
  particleSpawnScript [
   &InitializeParticle(
    Color=<rgba 0.01 0.01 0.01 0>
    Lifetime Max=1.1
    Lifetime Min=0.9
    Lifetime Mode=::NewEnumerator1
    Sprite Size=&Multiply_Vector2D_ByFloat(
     Float=&FloatFromCurve(FloatCurve=DataInterface() {propertyValues {Curve {keys [<at 0 1 RCIM_Constant>,<at 0.5 -1>]}}}, CurveIndex=$Emitter.LoopedAge)
     Vector2D=<2 2>
    )
    Sprite Size Mode=::NewEnumerator4
    UsePositionOffset=false
    Write Material Random=true
   )
   &SphereLocation(
    MaskSpawnGroup=false
    Non Uniform Scale=<4 0 1.5>
    Override Seed=false
    Sphere Radius=&Multiply_Float(A=8, B=$User.GlobalScale)
    Surface Only=false
    UseNonUniformScale=true
    UseOffset=false
    UseSphereOrientation=false
   )
   &AddVelocityFromPoint(
    Module.Use Falloff Distance=false
    Velocity Strength=&Multiply_Float(
     A=&Sine(Normalized Angle=$Emitter.Age, Period=1, Scale=60, Bias=0)
     B=&MakeFloatFromLinearColor(LinearColor=$System.Debug_AudioReact_Spectrum, Channel:ENiagaraLinearColor_Channels=::NewEnumerator2)
    )
   )
  ]
  particleUpdateScript [
   &ParticleState()
   &SolveForcesAndVelocity(Clamp Velocity=false, Limit Acceleration=false)
  ]
  rendererClasses [&NiagaraSpriteRendererProperties]
  renderers [{rendererClass=&NiagaraSpriteRendererProperties}]
 }
 PresetFX_Glow {
  emitterUpdateScript [
   &EmitterState(
    Inactive Response=::NewEnumerator1
    Life Cycle Mode=::NewEnumerator1
    Loop Behavior=::NewEnumerator1
    Loop Duration Mode=::NewEnumerator1
    UseLoopDelay=false
   )
   &SpawnBurst_Instantaneous(Use Spawn Probability=false, UseLoopCountLimit=false) {enabled=false}
   &SpawnRate(SpawnRate=12, Use Spawn Probability=false)
  ]
  particleSpawnScript [
   &InitializeParticle(
    Color=&ColorFromCurve(
     ColorCurve=DataInterface() {propertyValues {
       AlphaCurve {keys [<at 0 0.25 RCIM_Cubic>]}
       BlueCurve {keys [
         {interpMode=RCIM_Constant, leaveTangentWeight=0.0335, time=0, value=0.9104}
         <at 0.3132 1 RCIM_Constant>
         <at 0.5124 1 RCIM_Constant>
         <at 0.8917 0.5391 RCIM_Constant>
        ]}
       GreenCurve {keys [
         {interpMode=RCIM_Constant, leaveTangentWeight=0.1165, time=0, value=1}
         <at 0.3132 0 RCIM_Constant>
         <at 0.5124 0.2264 RCIM_Constant>
         <at 0.8917 0.0781 RCIM_Constant>
        ]}
       RedCurve {keys [
         {interpMode=RCIM_Constant, leaveTangentWeight=0.0645, time=0, value=0}
         <at 0.3132 0.6177 RCIM_Constant>
         <at 0.5124 0 RCIM_Constant>
         <at 0.8917 1 RCIM_Constant>
        ]}
      }}
     CurveIndex=&RandomRangeFloat(Minimum=0, Maximum=1)
    )
    Lifetime Max=1
    Lifetime Min=0.8
    Lifetime Mode=::NewEnumerator1
    Sprite Size Mode=::NewEnumerator1
    Uniform Sprite Size Max=&Multiply_Float(A=35, B=$User.GlobalScale)
    Uniform Sprite Size Min=&Multiply_Float(A=25, B=$User.GlobalScale)
    UsePositionOffset=true
    Write Material Random=true
   )
   &SphereLocation(
    MaskSpawnGroup=false
    Non Uniform Scale=<0.7 0 0.25>
    Override Seed=false
    Sphere Radius=&Multiply_Float(A=40, B=$User.GlobalScale)
    Surface Only=false
    UseNonUniformScale=true
    UseOffset=false
    UseSphereOrientation=false
   )
  ]
  particleUpdateScript [
   &ParticleState()
   &ScaleColor(
    Scale Alpha=&FloatFromCurve(
     FloatCurve=DataInterface() {propertyValues {Curve {keys [
         {arriveTangent=3.6277, interpMode=RCIM_Cubic, leaveTangent=3.6277, tangentMode=RCTM_User, time=0, value=0}
         {interpMode=RCIM_Cubic, tangentMode=RCTM_User, time=0.5, value=1}
         {tangentMode=RCTM_User, time=1, value=0}
        ]}}}
     Scale Curve=0.3
    )
    ScaleA=true
    ScaleRGB=false
   )
   &CameraOffset(Camera Offset Amount=-10)
  ]
  rendererClasses [&NiagaraSpriteRendererProperties]
  renderers [{rendererClass=&NiagaraSpriteRendererProperties}]
 }
 PresetFX_Glow001 {
  simTarget=GPUComputeSim
  emitterUpdateScript [
   &EmitterState(
    Inactive Response=::NewEnumerator1
    Life Cycle Mode=::NewEnumerator1
    Loop Behavior=::NewEnumerator1
    Loop Duration Mode=::NewEnumerator1
    UseLoopDelay=false
   )
   &SpawnBurst_Instantaneous(Use Spawn Probability=false, UseLoopCountLimit=false) {enabled=false}
   &SpawnRate(SpawnRate=7, Use Spawn Probability=false)
  ]
  particleSpawnScript [
   &InitializeParticle(
    Color=&ColorFromCurve(
     ColorCurve=DataInterface() {propertyValues {
       AlphaCurve {keys [<at 0 0.4 RCIM_Constant>]}
       BlueCurve {keys [
         {interpMode=RCIM_Constant, leaveTangentWeight=0.0335, time=0, value=0.9177}
         <at 0.1945 1 RCIM_Constant>
         <at 0.382 1 RCIM_Constant>
         <at 0.6556 1 RCIM_Constant>
         <at 0.8917 0.5391 RCIM_Constant>
        ]}
       GreenCurve {keys [
         {interpMode=RCIM_Constant, leaveTangentWeight=0.1165, time=0, value=1}
         <at 0.1945 0.2595 RCIM_Constant>
         <at 0.382 0.3084 RCIM_Constant>
         <at 0.6556 0.7846 RCIM_Constant>
         <at 0.8917 0.0781 RCIM_Constant>
        ]}
       RedCurve {keys [
         {interpMode=RCIM_Constant, leaveTangentWeight=0.0645, time=0, value=0.0816}
         <at 0.1945 0.7329 RCIM_Constant>
         <at 0.382 0.1059 RCIM_Constant>
         <at 0.6556 0.7003 RCIM_Constant>
         <at 0.8917 1 RCIM_Constant>
        ]}
      }}
     CurveIndex=&RandomRangeFloat(Minimum=0, Maximum=1)
    )
    Lifetime Mode=::NewEnumerator1
    Sprite Size Mode=::NewEnumerator1
    Uniform Sprite Size Max=&Multiply_Float(A=35, B=$User.GlobalScale)
    Uniform Sprite Size Min=&Multiply_Float(A=25, B=$User.GlobalScale)
    UsePositionOffset=true
    Write Material Random=true
   )
   &SphereLocation(
    MaskSpawnGroup=false
    Non Uniform Scale=<0.62 0 0.25>
    Offset=<0 -2 0>
    Override Seed=false
    Sphere Radius=&Multiply_Float(A=40, B=$User.GlobalScale)
    Surface Only=false
    UseNonUniformScale=true
    UseOffset=true
    UseSphereOrientation=false
   )
  ]
  particleUpdateScript [
   &ParticleState()
   &ScaleColor(
    Scale Alpha=&FloatFromCurve(
     FloatCurve=DataInterface() {propertyValues {Curve {keys [
         {arriveTangent=3.6277, interpMode=RCIM_Cubic, leaveTangent=3.6277, tangentMode=RCTM_User, time=0, value=0}
         {interpMode=RCIM_Cubic, tangentMode=RCTM_User, time=0.5, value=1}
         {tangentMode=RCTM_User, time=1, value=0}
        ]}}}
     Scale Curve=&Multiply_Float(A=1, B=1)
    )
    Scale RGB=<3 3 3>
    ScaleA=true
    ScaleRGB=true
   )
   &CameraOffset(Camera Offset Amount=-20)
  ]
  rendererClasses [&NiagaraSpriteRendererProperties]
  renderers [{rendererClass=&NiagaraSpriteRendererProperties}]
 }
}
summary {
 systemName=NS_BP_F_Commando_Zest
 emitters [
  {emitterName=Core, rendererClasses [&NiagaraMeshRendererProperties, &NiagaraLightRendererProperties]}
  {emitterName=PresetFX_Glow, rendererClasses [&NiagaraSpriteRendererProperties]}
  {emitterName=ParentEmitter, simTarget=GPUComputeSim, rendererClasses [&NiagaraSpriteRendererProperties]}
  {emitterName=ChildEmitter, simTarget=GPUComputeSim, rendererClasses [&NiagaraSpriteRendererProperties]}
  {emitterName=LineSprite, rendererClasses [&NiagaraSpriteRendererProperties]}
  {emitterName=PresetFX_Glow001, simTarget=GPUComputeSim, rendererClasses [&NiagaraSpriteRendererProperties]}
  {emitterName=Colorful_Quads, rendererClasses [&NiagaraSpriteRendererProperties]}
  {emitterName=Flash, rendererClasses [&NiagaraSpriteRendererProperties]}
  {emitterName=Core002, rendererClasses [&NiagaraMeshRendererProperties]}
  {emitterName=ChildEmitter001, rendererClasses [&NiagaraSpriteRendererProperties]}
 ]
 userVariables [
  User.Backpack_DiffuseTex {
   type=Texture
   description=""
   defaultValue {
    struct=NiagaraExt_VariableValue_Object
    value {
     object=/BRCosmetics/Accessories/FORT_Backpacks/Backpack_F_MED_Zest/FX/Texture/T_BP_FemaleCommandoZest_Base
     objectClass=/Script/Engine.Texture
    }
   }
  }
  User.Backpack_Mat {
   type=MaterialInterface
   description=""
   defaultValue {
    struct=NiagaraExt_VariableValue_Object
    value {
     object=/BRCosmetics/Accessories/FORT_Backpacks/Backpack_F_MED_Zest/FX/Material/MI_3DHB_Post_Zest
     objectClass=/Script/Engine.MaterialInterface
    }
   }
  }
  User.Backpack_NormalTex {
   type=Texture
   description=""
   defaultValue {
    struct=NiagaraExt_VariableValue_Object
    value {object=/BRCosmetics/Accessories/FORT_Backpacks/Backpack_F_MED_Zest/FX/Texture/T_3DHB_Normal, objectClass=&Texture}
   }
  }
  User.Backpack_WPOTex {
   type=Texture
   description=""
   defaultValue {
    struct=NiagaraExt_VariableValue_Object
    value {object=/BRCosmetics/Accessories/FORT_Backpacks/Backpack_F_MED_Zest/FX/Texture/T_3DHB_WPO, objectClass=&Texture}
   }
  }
  User.bIsFrontEnd {type=NiagaraBool}
  User.bIsFrontEndPreview {type=NiagaraBool}
  User.Debug_AudioREact {type=NiagaraFloat, defaultValue=0, description=""}
  User.GlobalScale {type=NiagaraFloat, defaultValue=0.75, description=""}
  User.IsFrom3D {type=NiagaraFloat, defaultValue=1, description=""}
  User.Ofs_from_Body {type=NiagaraFloat, defaultValue=0, description=""}
  User.PlayerSpeed {type=NiagaraFloat, defaultValue=0, description=""}
  User.SkeletalMesh {
   type=NiagaraDataInterfaceSkeletalMesh
   defaultValue {
    struct=NiagaraExt_VariableValue_DataInterface
    value {dataInterface=&self:NiagaraDataInterfaceSkeletalMesh_0}
   }
  }
 ]
}
