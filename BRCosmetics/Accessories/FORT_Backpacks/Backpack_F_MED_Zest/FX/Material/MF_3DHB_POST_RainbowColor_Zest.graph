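# Material function for the Zest backpack cosmetic that outputs an animated, view-dependent rainbow color for blending into emissive.
# It dots the camera direction (world position plus WPO input) with a flattened, transformed tangent-space normal, then adds time-driven flow.
# Density scales that value into a phase-shifted cosine palette, which is desaturated and multiplied by strength.
# Flow speed is an internal parameter; the other RainbowColor parameters are unconnected, with function inputs used instead.
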MaterialFunction /BRCosmetics/Accessories/FORT_Backpacks/Backpack_F_MED_Zest/FX/Material/MF_3DHB_POST_RainbowColor_Zest
graph {nodes [
  n1 = Time() {bOverride_Period=true, period=6000}
  n2 = FunctionInput() {bUsePreviewValueAsDefault=true, inputName=WPOIn}
  n3 = WorldPosition() {worldPositionShaderOffset=WPT_ExcludeAllShaderOffsets}
  n4 = Normalize(VectorInput=Subtract(A=CameraPositionWS(), B=Add(A=@n2, B=@n3.XYZ)))
  n5 = MaterialFunctionCall(
   Normal=FunctionInput() {bUsePreviewValueAsDefault=true, inputName="NormalIn(TS)", previewValue=<0 0 1 1>}
   Flatness=FunctionInput() {inputName=In_RainbowColor_FlattenNormal, inputType=FunctionInput_Scalar, previewValue=<0.8 0 0 1>}
  ) {materialFunction=&FlattenNormal}
  n6 = Add(
   A=Multiply(
    A=FunctionInput(Preview=@n1) {bUsePreviewValueAsDefault=true, inputName=Time, inputType=FunctionInput_Scalar}
    B=ScalarParameter() {defaultValue=-0.015, group=POST_RainbowColor, parameterName=RainbowColor_FlowSpeed}
   )
   B=DotProduct(A=FunctionInput(Preview=@n4) {bUsePreviewValueAsDefault=true, inputName=DirctionIn}, B=Transform(None=@n5.Result))
  )
  n7 = FunctionInput() {inputName=In_RainbowColor_Density, inputType=FunctionInput_Scalar, previewValue=<15 0 0 1>}
  n8 = Add(
   A=Multiply(A=Constant3Vector() {constant=<rgba 1 1 1 1>}, B=Multiply(A=@n6, B=@n7) {constB=2})
   B=Constant3Vector() {constant=<rgba 0 0.33 0.67 1>}
  )
  n9 = Add(A=Constant3Vector() {constant=<rgba 0.5 0.5 0.5 1>}, B=Multiply(A=Constant3Vector() {constant=<rgba 0.5 0.5 0.5 1>}, B=Cosine(None=@n8)))
  n10 = FunctionInput() {inputName=In_RainbowColor_Desaturation, inputType=FunctionInput_Scalar}
  FunctionOutput(None=Multiply(
    A=Desaturation(None=@n9, Fraction=@n10)
    B=FunctionInput() {inputName=In_RainbowColor_Strength, inputType=FunctionInput_Scalar, previewValue=<0.05 0 0 1>}
   )) {outputName=BlendToEmissive}
  ScalarParameter() {defaultValue=0.05, group=POST_RainbowColor, parameterName=RainbowColor_Strength}
  ScalarParameter() {group=POST_RainbowColor, parameterName=RainbowColor_Desaturation}
  ScalarParameter() {defaultValue=0.8, group=POST_RainbowColor, parameterName=RainbowColor_FlattenNormal}
  ScalarParameter() {defaultValue=15, group=POST_RainbowColor, parameterName=RainbowColor_Density}
 ]}
