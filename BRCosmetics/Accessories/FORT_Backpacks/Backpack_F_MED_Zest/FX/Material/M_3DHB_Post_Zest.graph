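# Translucent, unlit, two-sided 3DHB hologram material for the Zest backpack cosmetic's effects, usable on Niagara mesh particles.
# Base shader output is tinted by particle color, then gains noise-jittered chromatic aberration (skipped at low quality), rainbow color and rim light.
# Named reroutes link these stages and a shared hologram FX function, with dynamic parameters supplying per-particle control.
# It also reads a music spectrum value from an audio-reactive parameter collection.
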Material /BRCosmetics/Accessories/FORT_Backpacks/Backpack_F_MED_Zest/FX/Material/M_3DHB_Post_Zest
graph {
 properties {
  allowTranslucentCustomDepthWrites=true
  bUsedWithNiagaraMeshParticles=true
  blendMode=BLEND_Translucent
  ditherOpacityMask=true
  opacityMaskClipValue=0.05
  pixelDepthOffsetMode=PDOM_Legacy
  shadingModel=MSM_Unlit
  twoSided=true
  usedShadingModels=Unlit
 }
 nodes [
  n1 = DynamicParameter() {defaultValue=<rgba 1 1 0 0.01>, paramNames [Random, MeshScale, Flicker, chromatic aberration]}
  n2 = Add(A=Time() {bOverride_Period=true, period=6000}, B=@n1.Random)
  MaterialFunctionCall(Emissive=MaterialFunctionCall!(
    Time=@n2
    MeshScaleIn=@n1.MeshScale
    UVForScanline
    CustomFlicker=NamedRerouteUsage()
    UseCustomFlicker=ScalarParameter() {parameterName=IS UseCustomFlicker}
    UVForStroke
    UVForGlitch
    BaseUVIn
    EmissiveIn=NamedRerouteUsage()
    OpacityIn=NamedRerouteUsage()
    WPOIn=NamedRerouteUsage()
   ) {materialFunction=/Game/Tools/3DHB/Shader/Base/MF_3DHB_HologramFX}.EmissiveOut) {materialFunction=&MF_FX_EmissiveAdjustments}
  Multiply(A=NamedRerouteUsage(), B=VertexColor().R)
  n3 = MaterialFunctionCall!(MeshScaleIn=@n1.MeshScale) {materialFunction=/Game/Tools/3DHB/Shader/Base/MF_3DHB_Base}
  NamedRerouteDeclaration(None=ComponentMask(None=@n3.OriginDiffuse) {b=true, g=true, r=true}) {name=Base_Emissive, nodeColor=<rgba 0.5703 1 0 1>}
  NamedRerouteDeclaration(None=ComponentMask(None=@n3.OriginDiffuse) {a=true}) {name=Base_Opacity, nodeColor=<rgba 0.5703 1 0 1>}
  NamedRerouteDeclaration(None=@n3.OriginWPO) {name=Base_WPO, nodeColor=<rgba 0.5703 1 0 1>}
  NamedRerouteDeclaration(None=@n3.OriginNormal) {name=Base_Normal, nodeColor=<rgba 0.5703 1 0 1>}
  n4 = Multiply(A=ComponentMask(None=NamedRerouteUsage()) {b=true}, B=ParticleColor().RGB)
  n5 = TextureCoordinate()
  n6 = Multiply(A=NamedRerouteUsage()) {constB=0.7}
  n7 = ScalarParameter() {defaultValue=0.02, parameterName=AberrationOffset}
  n8 = Multiply(A=Time()) {constB=18000}
  n9 = WorldPosition()
  n10 = VectorNoise(World Position=Multiply(A=Add(A=@n8, B=@n9.XYZ)) {constB=0.001})
  n11 = Subtract(A=Multiply(A=ComponentMask(None=@n10) {g=true, r=true}) {constB=2})
  n12 = Add(
   A=Multiply(A=Multiply(A=@n6, B=@n7), B=@n11)
   B=Multiply(
    A=@n11
    B=DynamicParameter() {defaultValue=<rgba 1 1 0 0.01>, paramNames [Random, MeshScale, Flicker, chromatic aberration]}.chromatic aberration
   ) {constB=0.006}
  )
  n13 = TextureObjectParameter() {group=Base, parameterName=Base_DiffuseTexture}
  n14 = TextureSample(UVs=Add(A=@n5, B=@n12), Tex=@n13)
  n15 = Subtract(A=Subtract(A=@n14.R, B=@n14.R)) {constB=0.2}
  n16 = TextureSample(UVs=Subtract(A=@n5, B=@n12), Tex=@n13)
  n17 = Subtract(A=Subtract(A=@n16.R, B=@n16.R)) {constB=0.2}
  n18 = Add(
   A=@n4
   B=Multiply(
    A=MaterialFunctionCall(X=@n15, Y=@n17, Z=@n17) {materialFunction=&MakeFloat3}.Result
    B=VectorParameter() {parameterName=ColorAdjustment, defaultValue=<rgba 0.2614 0 1 0>}.RGB
   )
  ) {constB=0}
  n19 = QualitySwitch(Default=@n4, Low=@n4, High=@n18, Medium=@n18, Epic=@n18)
  n20 = NamedRerouteUsage()
  n21 = DynamicParameter() {
   parameterIndex=1
   defaultValue=<rgba 0.2 2 11 2>
   paramNames [RainbowColor_Strength, RainbowColor_Desaturation, RainbowColor_Density, RainbowColor_FlattenNormal]
  }
  n22 = NamedRerouteUsage()
  n23 = MaterialFunctionCall!(
   Time=NamedRerouteUsage()
   "NormalIn(TS)"=@n20
   DirctionIn
   In_RainbowColor_Strength=@n21.RainbowColor_Strength
   In_RainbowColor_Desaturation=@n21.RainbowColor_Strength
   In_RainbowColor_Density=@n21.RainbowColor_Strength
   In_RainbowColor_FlattenNormal=@n21.RainbowColor_Strength
   WPOIn=@n22
  ) {materialFunction=&/MF_3DHB_POST_RainbowColor_Zest}
  NamedRerouteDeclaration(None=Add(
    A=Add(A=@n19, B=@n23.BlendToEmissive)
    B=MaterialFunctionCall!("NormalIn(TS)"=@n20) {materialFunction=/BRCosmetics/Tools/3DHB/Shader/MasterLib/PostFunctions/MF_3DHB_POST_RimLight}.AddToEmissive
   )) {name=Post_Emissive, nodeColor=<rgba 0.5703 1 0 1>}
  n24 = NamedRerouteUsage()
  n25 = NamedRerouteUsage()
  NamedRerouteDeclaration(None=Multiply(A=Add(A=@n24, B=@n25), B=ParticleColor().A)) {name=Post_Opacity, nodeColor=<rgba 0.5703 1 0 1>}
  NamedRerouteDeclaration(None=@n22) {name=Post_WPO, nodeColor=<rgba 0.5703 1 0 1>}
  NamedRerouteDeclaration(None=@n2) {name=RandomTime, nodeColor=<rgba 0 1 0.7422 1>}
  NamedRerouteDeclaration(None=MaterialFunctionCall(Float4=CollectionParameter() {collection=/Game/Athena/Events/AudioAnalysis/MPC_Audio_Reactive_Parameters, parameterName=Spectrum_Music}) {materialFunction=&BreakOutFloat4Components}.A) {name=MPC_Spectrum_Music_A, nodeColor=<rgba 0.5703 1 0 1>}
  n26 = Add(A=@n15, B=@n17)
  NamedRerouteDeclaration(None=Saturate(None=Multiply(A=@n26) {constB=1.5})) {name=chromatic aberration, nodeColor=<rgba 0.5703 1 0 1>}
 ]
}
