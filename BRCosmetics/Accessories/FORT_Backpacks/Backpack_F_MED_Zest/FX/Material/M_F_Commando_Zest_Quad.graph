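# Unlit translucent material for Niagara sprite effects on the Zest backpack cosmetic.
# It multiplies particle color RGB by the RGB of a linear color curve texture borrowed from the shared spline trail effects, sampled with default UVs.
# The product feeds the Emissive input of the shared MF_FX_EmissiveAdjustments function.
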Material /BRCosmetics/Accessories/FORT_Backpacks/Backpack_F_MED_Zest/FX/Material/M_F_Commando_Zest_Quad
graph {
 properties {bUsedWithNiagaraSprites=true, blendMode=BLEND_Translucent, pixelDepthOffsetMode=PDOM_Legacy, shadingModel=MSM_Unlit, usedShadingModels=Unlit}
 nodes [MaterialFunctionCall(Emissive=Multiply(
    A=ParticleColor().RGB
    B=TextureSample() {samplerType=SAMPLERTYPE_LinearColor, texture=/Game/Effects/SplineTrails/Textures/T_LinearColorCurve_Default_01}.RGB
   )) {materialFunction=&MF_FX_EmissiveAdjustments}]
}
