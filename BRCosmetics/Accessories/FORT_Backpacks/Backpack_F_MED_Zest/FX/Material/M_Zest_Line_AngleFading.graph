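# Translucent unlit Niagara sprite material for line effects on the Zest backpack cosmetic, with depth test disabled.
# Two panning tiled smoke textures blend into noise, which particle alpha and a soft vertical UV gradient shape into opacity.
# Opacity fades as the camera-to-object direction nears world vertical, then passes through DepthFade.
# Emissive multiplies noise, particle color and a partial time-of-day tint, then runs through the shared FX emissive adjustment function.
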Material /BRCosmetics/Accessories/FORT_Backpacks/Backpack_F_MED_Zest/FX/Material/M_Zest_Line_AngleFading
graph {
 properties {
  bDisableDepthTest=true
  bUsedWithNiagaraSprites=true
  blendMode=BLEND_Translucent
  pixelDepthOffsetMode=PDOM_Legacy
  refractionMethod=RM_IndexOfRefraction
  shadingModel=MSM_Unlit
  usedShadingModels=Unlit
 }
 nodes [
  n1 = ParticleColor()
  n2 = TextureCoordinate() {uTiling=8, vTiling=0.35}
  n3 = TextureCoordinate() {uTiling=6, vTiling=0.2}
  n4 = LinearInterpolate(
   A=TextureSample(UVs=Panner(Coordinate=@n2) {speedX=0.1, speedY=0.7}) {texture=&T_Smooth_tile}.R
   B=TextureSample(UVs=Panner(Coordinate=@n3) {speedX=-0.1, speedY=1}) {texture=&T_Inky_Smoke_Tile}.R
  )
  n5 = TextureCoordinate()
  n6 = Multiply(A=OneMinus(None=ComponentMask(None=@n5) {g=true}), B=ComponentMask(None=@n5) {g=true})
  n7 = Multiply(A=Multiply(A=@n1.A, B=@n4), B=Power(Base=Multiply(A=@n6) {constB=2}))
  n8 = Subtract(A=CameraPositionWS(), B=ObjectPositionWS())
  n9 = Abs(None=DotProduct(A=Normalize(VectorInput=@n8), B=Constant3Vector() {constant=<rgba 0 0 1 0>}))
  DepthFade(Opacity=Multiply(A=@n7, B=OneMinus(None=@n9)), FadeDistance=ScalarParameter() {defaultValue=50, parameterName=DepthFade})
  LinearInterpolate(Alpha=@n7) {constA=1, constB=1.1}
  n10 = LinearInterpolate(B=MaterialFunctionCall(
    Color_Morning=VectorParameter() {parameterName=MorningTint, defaultValue=<rgba 1.15 0.8884 0.805 0>}.RGB
    Color_Day=VectorParameter() {parameterName=DayTint, defaultValue=<rgba 1.0812 1.2 1.1958 0>}.RGB
    Color_Evening=VectorParameter() {parameterName=EveningTint, defaultValue=<rgba 1.1 0.7845 0.77 1>}.RGB
    Color_Night=VectorParameter() {parameterName=NightTint, defaultValue=<rgba 0.2656 0.85 0.85 0>}.RGB
   ) {materialFunction=&Color_Time_of_Day}.Result) {constA=1}
  MaterialFunctionCall(Emissive=Multiply(A=Multiply(A=@n10, B=@n1.RGB), B=@n4)) {materialFunction=&MF_FX_EmissiveAdjustments}
 ]
}
